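# Control Rig for the FossilMech male medium body that runs the shared DeformRig function and corrects the left shoulder pad, hood and hood-string bones.
# Spherical pose readers on arm, clavicle, pad and head store weights in variables.
# Weighted parent constraints anchor accessory bones, then pose-weighted transform offsets adjust them.
# IsEmote, ADS and SkyDive switch corrections; emotes disable DeformRig's upper-arm option and constrain left upper-arm deform bones to the pad.
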ControlRigBlueprint /BRCosmetics/Characters/Player/Male/Medium/Bodies/M_MED_FossilMech/Meshes/M_MED_FossilMech_CtrlRig
graphs [
 RigUnit_SequenceAggregate_3_ContainedGraph {nodes [
   n1 = RigVMFunctionEntryNode()
   RigUnit_SequenceAggregate_2 = &RigVMFunction_Sequence(ExecuteContext=@n1.ExecuteContext)
   RigUnit_SequenceAggregate_1 = &RigVMFunction_Sequence(ExecuteContext=@RigUnit_SequenceAggregate_2.B)
   RigVMFunctionReturnNode(B=@RigUnit_SequenceAggregate_1.A, C=@RigUnit_SequenceAggregate_1.B, A=@RigUnit_SequenceAggregate_2.A)
  ]}
 RigVMFunctionLibrary {links [], nodes []}
 RigVMModel {nodes [
   RigUnit_BeginExecution = &RigUnit_BeginExecution()
   RigUnit_SequenceAggregate_1_1 = &RigVMFunction_Sequence(ExecuteContext=@RigUnit_BeginExecution.ExecutePin)
   RigUnit_SequenceAggregate_2 = RigVMAggregateNode(ExecuteContext=@RigUnit_SequenceAggregate_1_1.A)
   RigVMFunction_MathBoolNot = &RigVMFunction_MathBoolNot(Value=RigVMVariableNode() {variable=IsEmote}.Value)
   FN_DeformRig_FNC = RigVMFunctionReferenceNode(
    DeformRig_ Curve="(deformRig_l_upperarm_up_alpha_72_#1=False,deformRig_r_upperarm_up_alpha_73_#2=False,deformRig_l_upperarm_fr_alpha_74_#3=False,deformRig_r_upperarm_fr_alpha_75_CB51B8 ...(truncated)"
    DeformRig_Bool="(upperarm_1_#4=True,elbow_3_#5=True,wrist_7_#6=True,groin_8_#7=True,glute_17_#8=True,knee_9_40C408324CA261 ...(truncated)"
    DeformRig_Debug=(upperarm_debug_18_#4=False,wrist_debug_20_#6=False,lat_debug_26_#9=False,groinGlute_debug_27_#7=False)
    DeformRig_Value="(upperarm_up_18_#10=(Rotation=(X=0,Y=0,Z=0,W=1),Translation=(X=0,Y=0,Z=8),Scale3D=(X=0.7,Y=1,Z=1)),upperarm_fwd_fr_19_#11=(Rotation=(X=0,Y=0,Z=0,W=1),Translation=(X=0,Y=10,Z=0),Scale3D=(X=0.5,Y=1,Z= ...(truncated)"
    DeformRig_Bool.upperarm_1_#4=@RigVMFunction_MathBoolNot.Result
    ExecuteContext=@RigUnit_SequenceAggregate_2.A
   )
   n1 = &RigUnit_SphericalPoseReader(
    ActiveRegionScaleFactors=(PositiveWidth=0.25,NegativeWidth=0.75,PositiveHeight=0.25,NegativeHeight=0.75)
    ActiveRegionSize=0.1
    Debug=(bDrawDebug=false,DebugScale=25,DebugSegments=20,DebugThickness=0.25)
    DriverAxis=(X=1,Y=0,Z=0)
    DriverItem=(Type=Bone,Name="upperarm_l")
    FalloffRegionScaleFactors=(PositiveWidth=1,NegativeWidth=1,PositiveHeight=1,NegativeHeight=1)
    FalloffSize=0.45
    FlipWidthScaling=false
    OptionalParentItem=(Type=Bone,Name="spine_05")
    RotationOffset=(X=50,Y=0,Z=-2)
    ExecutePin=@RigUnit_SequenceAggregate_2.C
   )
   > RigVMVariableNode(Value=@n1.OutputParam) {variable=ArmUp_Adjust}
   > n2 = &RigUnit_SphericalPoseReader(
    ActiveRegionScaleFactors=(PositiveWidth=2,NegativeWidth=1,PositiveHeight=2,NegativeHeight=1)
    ActiveRegionSize=0.02
    Debug=(bDrawDebug=false,DebugScale=25,DebugSegments=20,DebugThickness=0.25)
    DriverAxis=(X=0,Y=1,Z=0)
    DriverItem=(Type=Bone,Name="dyn_shoulder_pad")
    FalloffRegionScaleFactors=(PositiveWidth=2,NegativeWidth=1,PositiveHeight=2,NegativeHeight=1)
    FalloffSize=0.1
    OptionalParentItem=(Type=Bone,Name="spine_05")
    RotationOffset=(X=100,Y=-120,Z=0)
   )
   > RigVMVariableNode(Value=@n2.OutputParam) {variable=ArmCrossing_Adjust}
   > n3 = &RigUnit_SphericalPoseReader(
    ActiveRegionScaleFactors=(PositiveWidth=0.25,NegativeWidth=0.25,PositiveHeight=0.25,NegativeHeight=0.25)
    ActiveRegionSize=0.1
    Debug=(bDrawDebug=false,DebugScale=25,DebugSegments=20,DebugThickness=0.25)
    DriverAxis=(X=-5,Y=0,Z=0)
    DriverItem=(Type=Bone,Name="upperarm_l")
    FalloffRegionScaleFactors=(PositiveWidth=0.5,NegativeWidth=0.5,PositiveHeight=0.5,NegativeHeight=0.5)
    FalloffSize=0.3
    OptionalParentItem=(Type=Bone)
    RotationOffset=(X=200,Y=-140,Z=0)
   )
   > RigVMVariableNode(Value=@n3.OutputParam) {variable=ADS_Adjust}
   > n4 = &RigUnit_SphericalPoseReader(
    ActiveRegionScaleFactors=(PositiveWidth=1,NegativeWidth=1,PositiveHeight=1,NegativeHeight=1)
    ActiveRegionSize=0.1
    Debug=(bDrawDebug=false,DebugScale=25,DebugSegments=20,DebugThickness=0.25)
    DriverAxis=(X=0,Y=1,Z=0)
    DriverItem=(Type=Bone,Name="head")
    FalloffRegionScaleFactors=(PositiveWidth=1,NegativeWidth=1.3,PositiveHeight=1.2,NegativeHeight=1)
    FalloffSize=0.2
    OptionalParentItem=(Type=Bone,Name="neck_01")
    RotationOffset=(X=90,Y=0,Z=-50)
   )
   > n5 = RigVMVariableNode(Value=@n4.OutputParam) {variable=HeadRot_Adjust}
   n6 = &RigVMFunction_ControlFlowBranch(ExecuteContext=@n5.ExecuteContext, Condition=RigVMVariableNode() {variable=IsEmote}.Value)
   n7 = &RigUnit_SphericalPoseReader(
    ActiveRegionScaleFactors=(PositiveWidth=1,NegativeWidth=1,PositiveHeight=1,NegativeHeight=1)
    ActiveRegionSize=0.1
    Debug=(bDrawDebug=false,DebugScale=25,DebugSegments=20,DebugThickness=0.25)
    DriverAxis=(X=1,Y=0,Z=0)
    DriverItem=(Type=Bone,Name="clavicle_l")
    FalloffRegionScaleFactors=(PositiveWidth=1,NegativeWidth=1,PositiveHeight=1,NegativeHeight=1)
    FalloffSize=0.2
    OptionalParentItem=(Type=Bone,Name="spine_05")
    RotationOffset=(X=0,Y=-30,Z=0)
    ExecutePin=@n6.True
   )
   > RigVMVariableNode(Value=@n7.OutputParam) {variable=EmoteArmsUp_Adjust}
   > n8 = &RigUnit_SphericalPoseReader(
    ActiveRegionScaleFactors=(PositiveWidth=1.5,NegativeWidth=1,PositiveHeight=1,NegativeHeight=1.25)
    ActiveRegionSize=0.05
    Debug=(bDrawDebug=true,DebugScale=25,DebugSegments=20,DebugThickness=0.25)
    DriverAxis=(X=0,Y=1,Z=0)
    DriverItem=(Type=Bone,Name="clavicle_l")
    FalloffRegionScaleFactors=(PositiveWidth=1.5,NegativeWidth=1,PositiveHeight=1,NegativeHeight=0.9)
    FalloffSize=0.15
    OptionalParentItem=(Type=Bone,Name="spine_05")
    RotationOffset=(X=115,Y=0,Z=0)
   )
   > RigVMVariableNode(Value=@n8.OutputParam) {variable=EmoteClavDown_Adjust}
   &RigUnit_ParentConstraint(
    AdvancedSettings=(RotationOrderForFilter=XZY)
    Child=(Type=Bone,Name="dyn_hood_string_base_r")
    Parents=((Item=(Type=Bone,Name="deform_pec_r"),Weight=0.0235),(Item=(Type=Bone,Name="spine_02"),Weight=0.0243),(Item=(Type=Bone,Name="clavicle_r"),Weight=0.1012),(Item=(Type=Bone,Name="spine_04"),Weight=0.3055),(Item=(Type=Bone,Name="spine_05"),Weight=0.5455))
    Weight=1
    ExecutePin=@RigUnit_SequenceAggregate_2.B
   )
   > &RigUnit_ParentConstraint(
    AdvancedSettings=(RotationOrderForFilter=XZY)
    Child=(Type=Bone,Name="dyn_hood_string_base_l")
    Parents=((Item=(Type=Bone,Name="deform_pec_l"),Weight=0.0234),(Item=(Type=Bone,Name="spine_02"),Weight=0.0248),(Item=(Type=Bone,Name="clavicle_l"),Weight=0.0986),(Item=(Type=Bone,Name="spine_04"),Weight=0.3117),(Item=(Type=Bone,Name="spine_05"),Weight=0.5415))
    Weight=1
   )
   > n9 = &RigUnit_ModifyTransforms(
    ItemToModify=((Item=(Type=Bone,Name="dyn_shoulder_pad"),Transform=(Rotation=(X=0,Y=0,Z=0,W=1),Translation=(X=0,Y=0,Z=0.64),Scale3D=(X=1,Y=1,Z=1))))
    Weight=1
   )
   n10 = &RigVMFunction_ControlFlowBranch(ExecuteContext=@n9.ExecutePin, Condition=RigVMVariableNode() {variable=IsEmote}.Value)
   &RigUnit_ParentConstraint(
    AdvancedSettings=(RotationOrderForFilter=XZY)
    Child=(Type=Bone,Name="dyn_shoulder_pad_front")
    Parents="((Item=(Type=Bone,Name=\"spine_04\"),Weight=0.0823),(Item=(Type=Bone,Name=\"upperarm_twist_01_l\"),Weight=0.1169),(Item=(Type=Bone,Name=\"spine_05\"),Weight=0.1319),(Item=(Type=Bone,Name=\"deform_upperarm_fr_l\"),Weight=0.1439),(Item=(Type=Bone,Name=\"deform_pec_l\" ...(truncated)"
    Weight=1
    ExecutePin=@n10.Completed
   )
   > &RigUnit_ParentConstraint(
    AdvancedSettings=(RotationOrderForFilter=XZY)
    Child=(Type=Bone,Name="dyn_shoulder_pad_back")
    Parents="((Item=(Type=Bone,Name=\"upperarm_twist_02_l\"),Weight=0.0171),(Item=(Type=Bone,Name=\"spine_04\"),Weight=0.0418),(Item=(Type=Bone,Name=\"spine_05\"),Weight=0.0572),(Item=(Type=Bone,Name=\"deform_upperarm_up_l\"),Weight=0.0762),(Item=(Type=Bone,Name=\"spine_03\"),We ...(truncated)"
    Weight=1
   )
   &RigUnit_ParentConstraint(
    AdvancedSettings=(RotationOrderForFilter=XZY)
    Child=(Type=Bone,Name="dyn_hood_shoulder_base_l")
    Parents=((Item=(Type=Bone,Name="dyn_shoulder_pad_front"),Weight=0.0175),(Item=(Type=Bone,Name="dyn_shoulder_pad_back"),Weight=0.0248),(Item=(Type=Bone,Name="dyn_shoulder_pad"),Weight=0.9577))
    Weight=1
    ExecutePin=@n10.True
   )
   > &RigUnit_ParentConstraint(
    AdvancedSettings=(RotationOrderForFilter=XZY)
    Child=(Type=Bone,Name="dyn_shoulder_pad")
    Parents="((Item=(Type=Bone,Name=\"spine_04\"),Weight=0.0252),(Item=(Type=Bone,Name=\"deform_upperarm_fr_l\"),Weight=0.0365),(Item=(Type=Bone,Name=\"upperarm_twist_01_l\"),Weight=0.1062),(Item=(Type=Bone,Name=\"deform_upperarm_up_l\"),Weight=0.1097),(Item=(Type=Bone,Name=\"s ...(truncated)"
    Weight=1
   )
   > &RigUnit_ParentConstraint(
    AdvancedSettings=(RotationOrderForFilter=XZY)
    Child=(Type=Bone,Name="deform_upperarm_bk_l")
    Parents=((Item=(Type=Bone,Name="dyn_shoulder_pad_front"),Weight=0.0013),(Item=(Type=Bone,Name="dyn_shoulder_pad"),Weight=0.4654),(Item=(Type=Bone,Name="dyn_shoulder_pad_back"),Weight=0.5333))
    Weight=1
   )
   > &RigUnit_ParentConstraint(
    AdvancedSettings=(RotationOrderForFilter=XZY)
    Child=(Type=Bone,Name="deform_upperarm_fr_l")
    Parents=((Item=(Type=Bone,Name="dyn_shoulder_pad_back"),Weight=0.0001),(Item=(Type=Bone,Name="dyn_shoulder_pad_front"),Weight=0.4814),(Item=(Type=Bone,Name="dyn_shoulder_pad"),Weight=0.5185))
    Weight=1
   )
   > &RigUnit_ParentConstraint(
    AdvancedSettings=(RotationOrderForFilter=XZY)
    Child=(Type=Bone,Name="deform_upperarm_up_l")
    Parents=((Item=(Type=Bone,Name="dyn_shoulder_pad_front"),Weight=0.0843),(Item=(Type=Bone,Name="dyn_shoulder_pad_back"),Weight=0.2782),(Item=(Type=Bone,Name="dyn_shoulder_pad"),Weight=0.6375))
    Weight=1
   )
   > &RigUnit_ModifyTransforms(
    ItemToModify=((Item=(Type=Bone,Name="deform_upperarm_up_l"),Transform=(Rotation=(X=0,Y=0,Z=0,W=1),Translation=(X=0,Y=0,Z=-2),Scale3D=(X=1,Y=1,Z=1))))
    Weight=1
   )
   > &RigUnit_ModifyTransforms(
    ItemToModify=((Item=(Type=Bone,Name="dyn_shoulder_pad"),Transform=(Rotation=(X=0,Y=0,Z=0,W=1),Translation=(X=0,Y=0,Z=0),Scale3D=(X=1.02,Y=1.02,Z=1.02))))
    Weight=1
   )
   RigUnit_SequenceAggregate_1 = &RigVMFunction_Sequence(ExecuteContext=@RigUnit_SequenceAggregate_1_1.B)
   n11 = &RigVMFunction_ControlFlowBranch(ExecuteContext=@RigUnit_SequenceAggregate_1.B, Condition=RigVMVariableNode() {variable=IsEmote}.Value)
   RigUnit_SequenceAggregate = &RigVMFunction_Sequence(ExecuteContext=@n11.Completed)
   n12 = &RigVMFunction_ControlFlowBranch(Condition=RigVMVariableNode() {variable=IsEmote}.Value, ExecuteContext=@RigUnit_SequenceAggregate.B)
   n13 = RigVMVariableNode() {variable=ArmCrossing_Adjust}
   &RigUnit_ModifyTransforms(
    ItemToModify=((Item=(Type=Bone,Name="dyn_hood_shoulder_base_l"),Transform=(Rotation=(X=0,Y=0,Z=0,W=1),Translation=(X=-2,Y=0,Z=2),Scale3D=(X=1,Y=1,Z=1))))
    ExecutePin=@n12.False
    Weight=@n13.Value
   )
   > &RigUnit_ModifyTransforms(
    ItemToModify=((Item=(Type=Bone,Name="dyn_hood_shoulder_01_l"),Transform=(Rotation=(X=0,Y=0,Z=0,W=1),Translation=(X=-2,Y=0,Z=0),Scale3D=(X=1,Y=1,Z=1))))
    Weight=@n13.Value
   )
   > &RigUnit_ModifyTransforms(
    ItemToModify=((Item=(Type=Bone,Name="dyn_hood_shoulder_02_l"),Transform=(Rotation=(X=0,Y=0,Z=0,W=1),Translation=(X=-2,Y=0,Z=0),Scale3D=(X=1,Y=1,Z=1))))
    Weight=@n13.Value
   )
   n14 = RigVMVariableNode() {variable=HeadRot_Adjust}
   &RigUnit_ModifyTransforms(
    ItemToModify=((Item=(Type=Bone,Name="dyn_hood_01"),Transform=(Rotation=(X=0,Y=0,Z=0,W=1),Translation=(X=15,Y=0,Z=-10),Scale3D=(X=1,Y=1,Z=1))))
    Weight=@n14.Value
    ExecutePin=@n12.Completed
   )
   > &RigUnit_ModifyTransforms(
    ItemToModify=((Item=(Type=Bone,Name="dyn_hood_02"),Transform=(Rotation=(X=0,Y=0,Z=0,W=1),Translation=(X=-15,Y=0,Z=5),Scale3D=(X=1,Y=1,Z=1))))
    Weight=@n14.Value
   )
   > &RigUnit_ModifyTransforms(
    ItemToModify=((Item=(Type=Bone,Name="dyn_hood_shoulder_r"),Transform=(Rotation=(X=0,Y=0,Z=0,W=1),Translation=(X=0,Y=0,Z=4),Scale3D=(X=1,Y=1,Z=1))))
    Weight=@n14.Value
   )
   n15 = &RigVMFunction_ControlFlowBranch(Condition=RigVMVariableNode() {variable=SkyDive}.Value, ExecuteContext=@RigUnit_SequenceAggregate.A)
   RigVMVariableNode(Value=1, ExecuteContext=@n15.True) {variable=ArmCrossing_Adjust}
   n16 = RigVMVariableNode() {variable=ArmUp_Adjust}
   &RigUnit_ModifyTransforms(
    ItemToModify=((Item=(Type=Bone,Name="dyn_hood_shoulder_base_l"),Transform=(Rotation=(X=0,Y=0,Z=0,W=1),Translation=(X=-1,Y=0,Z=2),Scale3D=(X=1,Y=1,Z=1))))
    Weight=@n16.Value
    ExecutePin=@n11.False
   )
   > &RigUnit_ModifyTransforms(
    ItemToModify=((Item=(Type=Bone,Name="dyn_hood_shoulder_01_l"),Transform=(Rotation=(X=0,Y=0,Z=0,W=1),Translation=(X=-5,Y=0,Z=0),Scale3D=(X=1,Y=1,Z=1))))
    Weight=@n16.Value
   )
   > &RigUnit_ModifyTransforms(
    ItemToModify=((Item=(Type=Bone,Name="dyn_hood_shoulder_02_l"),Transform=(Rotation=(X=0,Y=0,Z=0,W=1),Translation=(X=0,Y=0,Z=0),Scale3D=(X=1,Y=1,Z=1))))
    Weight=@n16.Value
   )
   n17 = &RigVMFunction_ControlFlowBranch(Condition=RigVMVariableNode() {variable=IsEmote}.Value, ExecuteContext=@RigUnit_SequenceAggregate_1.A)
   n18 = RigVMVariableNode() {variable=ArmUp_Adjust}
   &RigUnit_ModifyTransforms(
    ItemToModify=((Item=(Type=Bone,Name="dyn_shoulder_pad"),Transform=(Rotation=(X=0,Y=0,Z=0,W=1),Translation=(X=4,Y=-1.75,Z=0),Scale3D=(X=0.95,Y=0.95,Z=0.95))))
    ItemToModify.0.Transform.Rotation=&RigVMFunction_MathQuaternionFromEuler(Euler=(X=15,Y=0,Z=0)).Result
    ExecutePin=@n17.False
    Weight=@n18.Value
   )
   > &RigUnit_ModifyTransforms(
    ItemToModify=((Item=(Type=Bone,Name="dyn_shoulder_pad_back"),Transform=(Rotation=(X=0,Y=0,Z=0,W=1),Translation=(X=1.4,Y=-8,Z=2),Scale3D=(X=0.95,Y=0.95,Z=0.95))))
    ItemToModify.0.Transform.Rotation=&RigVMFunction_MathQuaternionFromEuler(Euler=(X=10,Y=0,Z=0)).Result
    Weight=@n18.Value
   )
   > &RigUnit_ModifyTransforms(
    ItemToModify=((Item=(Type=Bone,Name="dyn_shoulder_pad_front"),Transform=(Rotation=(X=0,Y=0,Z=0,W=1),Translation=(X=-2,Y=2,Z=0),Scale3D=(X=0.9,Y=0.9,Z=0.9))))
    ItemToModify.0.Transform.Rotation=&RigVMFunction_MathQuaternionFromEuler(Euler=(X=20,Y=0,Z=0)).Result
    Weight=@n18.Value
   )
   n19 = &RigVMFunction_ControlFlowBranch(Condition=RigVMVariableNode() {variable=IsEmote}.Value, ExecuteContext=@n17.Completed)
   n20 = &RigUnit_ModifyTransforms(
    ItemToModify=((Item=(Type=Bone,Name="dyn_shoulder_pad"),Transform=(Rotation=(X=0,Y=0,Z=0,W=1),Translation=(X=2,Y=0,Z=0),Scale3D=(X=1,Y=1,Z=1))))
    ExecutePin=@n19.Completed
    Weight=RigVMVariableNode() {variable=ArmCrossing_Adjust}.Value
   )
   n21 = &RigVMFunction_ControlFlowBranch(ExecuteContext=@n20.ExecutePin, Condition=RigVMVariableNode() {variable=ADS}.Value)
   &RigUnit_ModifyTransforms(
    ItemToModify=((Item=(Type=Bone,Name="dyn_shoulder_pad"),Transform=(Rotation=(X=0,Y=0,Z=0,W=1),Translation=(X=0.5,Y=0,Z=0),Scale3D=(X=1,Y=1,Z=1))))
    ExecutePin=@n21.True
    Weight=RigVMVariableNode() {variable=ADS_Adjust}.Value
   )
   n22 = &RigVMFunction_ControlFlowBranch(Condition=RigVMVariableNode() {variable=SkyDive}.Value, ExecuteContext=@n21.Completed)
   n23 = &RigVMFunction_ControlFlowBranch(ExecuteContext=@n22.Completed, Condition=RigVMVariableNode() {variable=IsEmote}.Value)
   &RigUnit_ModifyTransforms(
    ItemToModify=((Item=(Type=Bone,Name="dyn_shoulder_pad"),Transform=(Rotation=(X=0,Y=0,Z=0,W=1),Translation=(X=0.5,Y=0,Z=0.5),Scale3D=(X=0.9,Y=0.9,Z=0.9))))
    ExecutePin=@n23.True
    Weight=RigVMVariableNode() {variable=EmoteArmsUp_Adjust}.Value
   )
   RigVMFunction_MathQuaternionFromEuler = &RigVMFunction_MathQuaternionFromEuler(Euler=(X=0,Y=-15,Z=0))
   n24 = RigVMVariableNode() {variable=EmoteClavDown_Adjust}
   > &RigUnit_ModifyTransforms(
    ItemToModify=((Item=(Type=Bone,Name="dyn_shoulder_pad_front"),Transform=(Rotation=(X=0,Y=0,Z=0,W=1),Translation=(X=4,Y=0,Z=2),Scale3D=(X=1,Y=1,Z=1))))
    ItemToModify.0.Transform.Rotation=@RigVMFunction_MathQuaternionFromEuler.Result
    Weight=@n24.Value
   )
   > &RigUnit_ModifyTransforms(
    ItemToModify=((Item=(Type=Bone,Name="dyn_shoulder_pad"),Transform=(Rotation=(X=0,Y=0,Z=0,W=1),Translation=(X=0,Y=0,Z=0),Scale3D=(X=1,Y=1,Z=1))))
    ItemToModify.0.Transform.Rotation=&RigVMFunction_MathQuaternionFromEuler(Euler=(X=10,Y=0,Z=0)).Result
    Weight=@n24.Value
   )
   &RigUnit_ModifyTransforms(
    ItemToModify=((Item=(Type=Bone,Name="dyn_shoulder_pad"),Transform=(Rotation=(X=0,Y=0,Z=0,W=1),Translation=(X=2.5,Y=0,Z=0.5),Scale3D=(X=1,Y=1,Z=1))))
    Weight=1
    ExecutePin=@n22.True
   )
  ]}
]
hierarchy [
 root {loc=<0 0 0>, rot=<0 0 0 1>}
 pelvis {parent=root, loc=<0 -0.7071 93.1191>, rot=<0.0127 0.707 -0.0127 -0.707>}
 spine_01 {parent=pelvis, loc=<6.7494 -0.6388 0>, rot=<0 0 0.034 -0.9994>}
 spine_02 {parent=spine_01, loc=<7.2518 -0.035 0>, rot=<0 0 0.0238 -0.9997>}
 spine_03 {parent=spine_02, loc=<6.0104 0 0>, rot=<0 0 0.0149 -0.9999>}
 spine_04 {parent=spine_03, loc=<8.1024 0 0>, rot=<0 0 -0.0132 -0.9999>}
 spine_05 {parent=spine_04, loc=<10.0146 0 0>, rot=<0 0 -0.0015 -1>}
 attach_cape {parent=spine_05, loc=<9.017 -11.714 0>, rot=<0.0616 -0.7044 -0.0616 -0.7044>}
 neck_01 {parent=spine_05, loc=<11.3138 0.1529 0>, rot=<0 0 -0.2003 -0.9797>}
 neck_02 {parent=neck_01, loc=<4.5627 0 0>, rot=<0 0 0.0504 -0.9987>}
 head {parent=neck_02, loc=<5.3914 0 0>, rot=<0 0 0.115 -0.9934>}
 attach_backpack {parent=spine_05, loc=<-0.123 -13.473 0>, rot=<-0.0093 -0.707 0.0093 -0.707>}
 clavicle_l {parent=spine_05, loc=<4.3575 -0.1466 -2.3223>, rot=<0.0465 -0.7289 -0.009 -0.6829>}
 upperarm_l {parent=clavicle_l, loc=<15.5111 0 0>, rot=<0.0226 -0.3234 -0.0088 -0.9459>}
 lowerarm_l {parent=upperarm_l, loc=<26.6514 0 0>, rot=<0.0391 -0.0099 -0.2453 -0.9686>}
 hand_l {parent=lowerarm_l, loc=<22.3491 0 0>, rot=<-0.7081 -0.0507 0.0975 -0.6975>}
 thumb_01_l {parent=hand_l, loc=<1.3118 -0.3191 -2.459>, rot=<0.6546 -0.4042 0.0664 -0.6354>}
 thumb_02_l {parent=thumb_01_l, loc=<5.9336 -0.0448 0.0401>, rot=<-0.0037 0.0213 0.298 -0.9543>}
 thumb_03_l {parent=thumb_02_l, loc=<2.2223 0.3682 -0.0907>, rot=<0.0118 -0.013 0.0066 -0.9998>}
 index_metacarpal_l {parent=hand_l, loc=<1.6535 0.0183 -2.96>, rot=<0.0158 -0.0124 -0.0173 -0.9996>}
 index_01_l {parent=index_metacarpal_l, loc=<8.3008 0.0357 -0.0524>, rot=<-0.0044 0.0485 0.0924 -0.9945>}
 index_02_l {parent=index_01_l, loc=<3.3587 -0.3004 0.0452>, rot=<-0.002 0.0027 0.3225 -0.9466>}
 index_03_l {parent=index_02_l, loc=<2.5578 -0.123 -0.0252>, rot=<-0.002 0.0067 0.2558 -0.9667>}
 middle_metacarpal_l {parent=hand_l, loc=<2.1856 -0.1725 -0.5984>, rot=<0.006 0.0168 -0.0329 -0.9993>}
 middle_01_l {parent=middle_metacarpal_l, loc=<7.3275 -0.2447 0.0129>, rot=<-0.0208 0.0042 0.0858 -0.9961>}
 middle_02_l {parent=middle_01_l, loc=<4.0123 0.1401 0.1175>, rot=<-0.0218 0.0258 0.3512 -0.9357>}
 middle_03_l {parent=middle_02_l, loc=<2.6077 0.0534 0.1373>, rot=<0.0082 0.0188 0.1708 -0.9851>}
 ring_metacarpal_l {parent=hand_l, loc=<2.2074 -0.3625 1.3638>, rot=<-0.0284 0.0722 -0.0458 -0.9959>}
 ring_01_l {parent=ring_metacarpal_l, loc=<7.0295 -0.6197 -0.1446>, rot=<-0.0489 -0.0252 0.1892 -0.9804>}
 ring_02_l {parent=ring_01_l, loc=<3.6174 0.4647 -0.1027>, rot=<0 0 0.2806 -0.9598>}
 ring_03_l {parent=ring_02_l, loc=<2.4152 0.114 0.0199>, rot=<0 0 0.1189 -0.9929>}
 pinky_metacarpal_l {parent=hand_l, loc=<1.7616 -0.9753 3.3814>, rot=<-0.0397 0.0814 -0.069 -0.9935>}
 pinky_01_l {parent=pinky_metacarpal_l, loc=<7.1709 -0.4831 -0.1094>, rot=<-0.0654 -0.0035 0.2242 -0.9723>}
 pinky_02_l {parent=pinky_01_l, loc=<2.8274 0.0893 -0.1043>, rot=<0.0037 0.0011 0.2854 -0.9584>}
 pinky_03_l {parent=pinky_02_l, loc=<2.2046 0.0307 -0.0078>, rot=<-0.0002 0.0098 0.1445 -0.9895>}
 deform_wrist_dn_base_l {parent=hand_l, loc=<0 -2.2342 0>, rot=<0 0 0 -1>}
 deform_wrist_dn_l {parent=deform_wrist_dn_base_l, loc=<0 0 0>, rot=<0 0 0 1>}
 deform_wrist_up_base_l {parent=hand_l, loc=<0 2.2682 0>, rot=<0 0 0 -1>}
 deform_wrist_up_l {parent=deform_wrist_up_base_l, loc=<0 0 0>, rot=<0 0 0 1>}
 lowerarm_twist_02_l {parent=lowerarm_l, loc=<9.6199 0 0>, rot=<0 0 0 -1>}
 lowerarm_twist_01_l {parent=lowerarm_l, loc=<17.8549 0 0>, rot=<0 0 0 -1>}
 deform_elbow_in_base_l {parent=lowerarm_l, loc=<0.2112 5.259 0.8274>, rot=<0 0 0 -1>}
 deform_elbow_in_l {parent=deform_elbow_in_base_l, loc=<0 0 0>, rot=<0 0 0 1>}
 deform_elbow_out_base_l {parent=lowerarm_l, loc=<0 -4 0>, rot=<0 0 0 -1>}
 deform_elbow_out_l {parent=deform_elbow_out_base_l, loc=<0 0 0>, rot=<0 0 0 1>}
 upperarm_twist_01_l {parent=upperarm_l, loc=<0.0041 0 0>, rot=<0 0 0 1>}
 deform_upperarm_bk_base_l {parent=upperarm_twist_01_l, loc=<1.0813 -7.2193 -1.9453>, rot=<0 0 0 1>}
 deform_upperarm_bk_l {parent=deform_upperarm_bk_base_l, loc=<0 0 0>, rot=<0 0 0 1>}
 deform_upperarm_fr_base_l {parent=upperarm_twist_01_l, loc=<-0.3691 7.0506 -2.765>, rot=<0 0 0 1>}
 deform_upperarm_fr_l {parent=deform_upperarm_fr_base_l, loc=<0 0 0>, rot=<0 0 0 1>}
 deform_upperarm_up_base_l {parent=upperarm_twist_01_l, loc=<-3.2216 -1.1391 3.3311>, rot=<0 0 0 1>}
 deform_upperarm_up_l {parent=deform_upperarm_up_base_l, loc=<0 0 0>, rot=<0 0 0 1>}
 upperarm_twist_02_l {parent=upperarm_l, loc=<13.7941 0 0>, rot=<0 0 0 1>}
 dyn_shoulder_pad {parent=clavicle_l, loc=<10.4762 0.5165 3.1485>, rot=<0.0008 0.0336 -0.0252 -0.9991>}
 dyn_shoulder_pad_front {parent=dyn_shoulder_pad, loc=<-0.2418 8.7324 -1.7369>, rot=<0.0137 -0.0133 -0.718 -0.6958>}
 dyn_shoulder_pad_back {parent=dyn_shoulder_pad, loc=<-0.534 -7.7482 -3.8424>, rot=<0.0227 0.0207 0.7395 -0.6725>}
 dyn_hood_shoulder_l {parent=clavicle_l, loc=<6.5205 0.1335 10.0121>, rot=<0.0008 0.0336 -0.0252 -0.9991>}
 dyn_hood_shoulder_helper_l {parent=dyn_hood_shoulder_l, loc=<1.9345 0.1388 -0.6803>, rot=<0 0 0 1>}
 dyn_hood_shoulder_base_l {parent=clavicle_l, loc=<6.5205 0.1335 10.0121>, rot=<0.0008 0.0336 -0.0252 -0.9991>}
 dyn_hood_shoulder_01_l {parent=dyn_hood_shoulder_base_l, loc=<1.9345 0.1388 -0.6803>, rot=<0 0 0 1>}
 dyn_hood_shoulder_02_l {parent=dyn_hood_shoulder_base_l, loc=<1.7371 -5.5064 -0.5382>, rot=<0 0 0 1>}
 clavicle_r {parent=spine_05, loc=<4.3575 -0.1466 2.3223>, rot=<0.7289 0.0465 -0.6829 0.009>}
 upperarm_r {parent=clavicle_r, loc=<-15.5111 0 0>, rot=<0.0226 -0.3234 -0.0088 -0.9459>}
 lowerarm_r {parent=upperarm_r, loc=<-26.6514 0 0>, rot=<0.0391 -0.0099 -0.2453 -0.9686>}
 hand_r {parent=lowerarm_r, loc=<-22.3491 0 0>, rot=<-0.7081 -0.0507 0.0975 -0.6975>}
 thumb_01_r {parent=hand_r, loc=<-1.3118 0.3191 2.459>, rot=<0.6546 -0.4042 0.0664 -0.6354>}
 thumb_02_r {parent=thumb_01_r, loc=<-5.9336 0.0448 -0.0401>, rot=<-0.0037 0.0213 0.298 -0.9543>}
 thumb_03_r {parent=thumb_02_r, loc=<-2.2223 -0.3682 0.0907>, rot=<0.0118 -0.013 0.0066 -0.9998>}
 index_metacarpal_r {parent=hand_r, loc=<-1.6535 -0.0183 2.96>, rot=<0.0158 -0.0124 -0.0173 -0.9996>}
 index_01_r {parent=index_metacarpal_r, loc=<-8.3008 -0.0357 0.0524>, rot=<-0.0044 0.0485 0.0924 -0.9945>}
 index_02_r {parent=index_01_r, loc=<-3.3587 0.3004 -0.0452>, rot=<-0.002 0.0027 0.3225 -0.9466>}
 index_03_r {parent=index_02_r, loc=<-2.5578 0.123 0.0252>, rot=<-0.002 0.0067 0.2558 -0.9667>}
 middle_metacarpal_r {parent=hand_r, loc=<-2.1856 0.1725 0.5984>, rot=<0.006 0.0168 -0.0329 -0.9993>}
 middle_01_r {parent=middle_metacarpal_r, loc=<-7.3275 0.2447 -0.0129>, rot=<-0.0208 0.0042 0.0858 -0.9961>}
 middle_02_r {parent=middle_01_r, loc=<-4.0123 -0.1401 -0.1175>, rot=<-0.0218 0.0258 0.3512 -0.9357>}
 middle_03_r {parent=middle_02_r, loc=<-2.6077 -0.0534 -0.1373>, rot=<0.0082 0.0188 0.1708 -0.9851>}
 ring_metacarpal_r {parent=hand_r, loc=<-2.2074 0.3625 -1.3638>, rot=<-0.0284 0.0722 -0.0458 -0.9959>}
 ring_01_r {parent=ring_metacarpal_r, loc=<-7.0295 0.6197 0.1446>, rot=<-0.0489 -0.0252 0.1892 -0.9804>}
 ring_02_r {parent=ring_01_r, loc=<-3.6174 -0.4647 0.1027>, rot=<0 0 0.2806 -0.9598>}
 ring_03_r {parent=ring_02_r, loc=<-2.4152 -0.114 -0.0199>, rot=<0 0 0.1189 -0.9929>}
 pinky_metacarpal_r {parent=hand_r, loc=<-1.7616 0.9753 -3.3814>, rot=<-0.0397 0.0814 -0.069 -0.9935>}
 pinky_01_r {parent=pinky_metacarpal_r, loc=<-7.1709 0.4831 0.1094>, rot=<-0.0654 -0.0035 0.2242 -0.9723>}
 pinky_02_r {parent=pinky_01_r, loc=<-2.8274 -0.0893 0.1043>, rot=<0.0037 0.0011 0.2854 -0.9584>}
 pinky_03_r {parent=pinky_02_r, loc=<-2.2046 -0.0307 0.0078>, rot=<-0.0002 0.0098 0.1445 -0.9895>}
 deform_wrist_dn_base_r {parent=hand_r, loc=<-0.0004 2.2345 -0.0001>, rot=<0 0 0 -1>}
 deform_wrist_dn_r {parent=deform_wrist_dn_base_r, loc=<0 0 0>, rot=<0 0 0 1>}
 deform_wrist_up_base_r {parent=hand_r, loc=<0.0003 -2.2684 0.0001>, rot=<0 0 0 -1>}
 deform_wrist_up_r {parent=deform_wrist_up_base_r, loc=<0 0 0>, rot=<0 0 0 1>}
 lowerarm_twist_02_r {parent=lowerarm_r, loc=<-9.6199 0 0>, rot=<0 0 0 -1>}
 lowerarm_twist_01_r {parent=lowerarm_r, loc=<-17.8549 0 0>, rot=<0 0 0 -1>}
 deform_elbow_in_base_r {parent=lowerarm_r, loc=<-0.2111 -5.2591 -0.8275>, rot=<0 0 0 -1>}
 deform_elbow_in_r {parent=deform_elbow_in_base_r, loc=<-0.0001 0 0.0002>, rot=<0 0 0 1>}
 deform_elbow_out_base_r {parent=lowerarm_r, loc=<0 4 -0.0001>, rot=<0 0 0 -1>}
 deform_elbow_out_r {parent=deform_elbow_out_base_r, loc=<0 0 0>, rot=<0 0 0 1>}
 upperarm_twist_01_r {parent=upperarm_r, loc=<-0.0041 0 0>, rot=<0 0 0 -1>}
 deform_upperarm_bk_base_r {parent=upperarm_twist_01_r, loc=<-1.0814 7.2192 1.9455>, rot=<0 0 0 1>}
 deform_upperarm_bk_r {parent=deform_upperarm_bk_base_r, loc=<0 0 0>, rot=<0 0 0 1>}
 deform_upperarm_fr_base_r {parent=upperarm_twist_01_r, loc=<0.3689 -7.0506 2.7652>, rot=<0 0 0 1>}
 deform_upperarm_fr_r {parent=deform_upperarm_fr_base_r, loc=<0 0 0>, rot=<0 0 0 1>}
 deform_upperarm_up_base_r {parent=upperarm_twist_01_r, loc=<3.2214 1.1391 -3.3308>, rot=<0 0 0 1>}
 deform_upperarm_up_r {parent=deform_upperarm_up_base_r, loc=<0 0 0>, rot=<0 0 0 -1>}
 upperarm_twist_02_r {parent=upperarm_r, loc=<-13.7941 0 0>, rot=<0 0 0 -1>}
 dyn_hood_shoulder_r {parent=clavicle_r, loc=<-6.5205 -0.1335 -10.0123>, rot=<0.0008 0.0336 -0.0252 -0.9991>}
 dyn_hood_string_base_r {parent=spine_05, loc=<6.0647 12.0566 1.5853>, rot=<0.1373 0.6848 0.7034 0.1321>}
 dyn_hood_string_01_r {parent=dyn_hood_string_base_r, loc=<4.0671 0 0>, rot=<0.039 -0.1071 -0.0145 -0.9934>}
 dyn_hood_string_02_r {parent=dyn_hood_string_01_r, loc=<5.6506 0 0>, rot=<0.0006 -0.0752 -0.008 -0.9971>}
 dyn_hood_string_03_r {parent=dyn_hood_string_02_r, loc=<5.3885 0 0>, rot=<0.0088 -0.0455 -0.003 -0.9989>}
 dyn_hood_string_04_r {parent=dyn_hood_string_03_r, loc=<5.791 0 0>, rot=<0.0174 -0.0114 0.0008 -0.9998>}
 dyn_hood_string_05_r {parent=dyn_hood_string_04_r, loc=<4.6114 0 0>, rot=<0 -0.0263 -0.0009 -0.9997>}
 dyn_hood_string_base_l {parent=spine_05, loc=<6.0643 12.0566 -1.9854>, rot=<-0.1489 -0.6967 -0.6849 -0.1525>}
 dyn_hood_string_01_l {parent=dyn_hood_string_base_l, loc=<3.5818 0 0>, rot=<0.0198 -0.122 -0.0081 -0.9923>}
 dyn_hood_string__02_l {parent=dyn_hood_string_01_l, loc=<5.7825 0 0>, rot=<0.0683 -0.0829 -0.0126 -0.9941>}
 dyn_hood_string_03_l {parent=dyn_hood_string__02_l, loc=<5.0037 0 0>, rot=<-0.126 -0.0438 0.0069 -0.991>}
 dyn_hood_string_04_l {parent=dyn_hood_string_03_l, loc=<5.3508 0 0>, rot=<0.0434 -0.0249 0.0072 -0.9987>}
 dyn_hood_string_05_l {parent=dyn_hood_string_04_l, loc=<4.8773 0 0>, rot=<-0.0347 0.0167 0.0092 -0.9992>}
 dyn_hood_01 {parent=spine_05, loc=<14.4374 -11.1624 -0.0351>, rot=<0.0284 -0.7065 0.7065 -0.0284>}
 dyn_hood_02 {parent=dyn_hood_01, loc=<5.7623 0 0>, rot=<0.0081 -0.0306 0.0002 -0.9995>}
 deform_pec_aim_base_l {parent=spine_04, loc=<7.8732 12.6388 -2.4123>, rot=<0.0102 -0.4602 0.3979 -0.7936>}
 deform_pec_aim_l {parent=deform_pec_aim_base_l, loc=<0 0 0>, rot=<0 0 0 1>}
 deform_pec_base_l {parent=deform_pec_aim_l, loc=<10.766 0 0>, rot=<0 0 0 1>}
 deform_pec_l {parent=deform_pec_base_l, loc=<0.055 0 0>, rot=<0 0 0 1>}
 deform_pec_aim_base_r {parent=spine_04, loc=<7.8727 12.6388 2.4123>, rot=<-0.4602 -0.0102 0.7936 0.3979>}
 deform_pec_aim_r {parent=deform_pec_aim_base_r, loc=<0 0 0>, rot=<0 0 0 1>}
 deform_pec_base_r {parent=deform_pec_aim_r, loc=<-10.7662 -0.0006 -0.0007>, rot=<0 0 0 1>}
 deform_pec_r {parent=deform_pec_base_r, loc=<0 0 0>, rot=<0 0 0 1>}
 thigh_l {parent=pelvis, loc=<-2.5693 -0.0109 -7.7501>, rot=<0.0012 0.0572 0.0155 -0.9982>}
 calf_l {parent=thigh_l, loc=<-39.0484 0 0>, rot=<0.0006 -0.0161 -0.0391 -0.9991>}
 foot_l {parent=calf_l, loc=<-42.561 0.1077 0.4292>, rot=<0.0418 -0.0455 0.0343 -0.9975>}
 ball_l {parent=foot_l, loc=<-6.1599 13.8264 0.6806>, rot=<0.0001 -0.0001 -0.7063 -0.7079>}
 calf_twist_02_l {parent=calf_l, loc=<-18.8064 -0.0967 0.0397>, rot=<0.0237 0.0095 -0.0035 -0.9997>}
 calf_twist_01_l {parent=calf_l, loc=<-35.813 -0.0967 0.0397>, rot=<0.0237 0.0095 -0.0035 -0.9997>}
 deform_knee_in_base_l {parent=calf_l, loc=<1.298 -5.2957 -0.0496>, rot=<0 0 0 -1>}
 deform_knee_in_l {parent=deform_knee_in_base_l, loc=<0 0 0>, rot=<0 0 0 1>}
 deform_knee_out_base_l {parent=calf_l, loc=<0 4 0>, rot=<0 0 0 -1>}
 deform_knee_out_l {parent=deform_knee_out_base_l, loc=<0 0 0>, rot=<0 0 0 1>}
 thigh_twist_01_l {parent=thigh_l, loc=<-21.6303 0 0>, rot=<0.0013 0 0 -1>}
 deform_groin_base_l {parent=thigh_l, loc=<-2.1908 8.1142 -1.7442>, rot=<0 0 0 -1>}
 deform_groin_l {parent=deform_groin_base_l, loc=<0 0 0>, rot=<0 0 0 1>}
 thigh_r {parent=pelvis, loc=<-2.5693 -0.0109 7.7501>, rot=<0.0572 -0.0012 0.9982 0.0155>}
 calf_r {parent=thigh_r, loc=<39.0484 0 0>, rot=<0.0006 -0.0161 -0.0391 -0.9991>}
 foot_r {parent=calf_r, loc=<42.561 -0.1077 -0.4292>, rot=<0.0418 -0.0455 0.0343 -0.9975>}
 ball_r {parent=foot_r, loc=<6.1599 -13.8264 -0.6806>, rot=<0.0001 -0.0001 -0.7063 -0.7079>}
 calf_twist_02_r {parent=calf_r, loc=<18.8064 0.0967 -0.0397>, rot=<0.0237 0.0095 -0.0035 -0.9997>}
 calf_twist_01_r {parent=calf_r, loc=<35.813 0.0967 -0.0397>, rot=<0.0237 0.0095 -0.0035 -0.9997>}
 deform_knee_in_base_r {parent=calf_r, loc=<-1.298 5.2957 0.0496>, rot=<0 0 0 -1>}
 deform_knee_in_r {parent=deform_knee_in_base_r, loc=<0 0 0>, rot=<0 0 0 1>}
 deform_knee_out_base_r {parent=calf_r, loc=<0 -4 0>, rot=<0 0 0 -1>}
 deform_knee_out_r {parent=deform_knee_out_base_r, loc=<0 0 0>, rot=<0 0 0 1>}
 thigh_twist_01_r {parent=thigh_r, loc=<21.6303 0 0>, rot=<0.0013 0 0 -1>}
 deform_groin_base_r {parent=thigh_r, loc=<2.1908 -8.1142 1.7442>, rot=<0 0 0 -1>}
 deform_groin_r {parent=deform_groin_base_r, loc=<0 0 0>, rot=<0 0 0 1>}
 deform_glute_base_l {parent=pelvis, loc=<-8.8064 -8.8222 -8.4287>, rot=<0.0275 0.0501 -0.4472 -0.8926>}
 deform_glute_l {parent=deform_glute_base_l, loc=<0 0 0>, rot=<0 0 0 1>}
 deform_glute_base_r {parent=pelvis, loc=<-8.8062 -8.8222 8.4287>, rot=<0.0501 -0.0275 0.8926 -0.4472>}
 deform_glute_r {parent=deform_glute_base_r, loc=<0 0 0>, rot=<0 0 0 1>}
]
variables [
 ADS {type=bool, default=False}
 ADS_Adjust {type=double, default=0}
 ArmCrossing_Adjust {type=double, default=0}
 ArmUp_Adjust {type=double, default=0}
 EmoteArmsUp_Adjust {type=double, default=0}
 EmoteClavDown_Adjust {type=double, default=0}
 HeadRot_Adjust {type=double, default=0}
 IsEmote {type=bool, default=False}
 SkyDive {type=bool, default=False}
]
