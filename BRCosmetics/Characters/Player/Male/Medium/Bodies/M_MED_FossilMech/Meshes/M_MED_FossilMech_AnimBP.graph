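# Animation Blueprint for the FossilMech medium male player body cosmetic, adding secondary motion to its hood and drawstrings.
# It copies the attached parent's pose through a Control Rig fed targeting, skydive and emote states, then runs spring-driven AnimDynamics on both string chains, hood and shoulder bones.
# Skydiving lowers hood and shoulder simulation strength until landing or exit, while strings stay at full alpha.
# Wearing backbling offsets the hood bone; front-end loading sets a Sim Space variable to World.
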AnimBlueprint /BRCosmetics/Characters/Player/Male/Medium/Bodies/M_MED_FossilMech/Meshes/M_MED_FossilMech_AnimBP
target_skeleton=&/M_MED_FossilMech_Skeleton
graphs [
 AnimGraph {nodes [
   n1 = Get AnimDyn Strength() -> (AnimDyn Strength:double)
   n2 = Control Rig!(
    Source=Copy Pose From Mesh() {bUseAttachedParent=true}
    Alpha
    ADS:bool=Get bIsTargeting()
    SkyDive:bool=Get bIsSkydiveFloating()
    IsEmote:bool=Get bIsPlayingEmote()
   ) -> (Pose) {
    alphaBoolBlend {blendOption=Linear}
    controlRigAssetReference {blueprintRigClass=&/M_MED_FossilMech_CtrlRig.M_MED_FossilMech_CtrlRig_C}
   }
   n3 = Anim Dynamics - Bone: dyn_hood_string_01_r!(SimulationSpace:AnimPhysSimSpaceType Enum, ComponentPose=Local To Component(LocalPose=@n2), Alpha=1, ExternalForce="0,0,0") -> (Pose) {
    angularDampingOverride=1
    angularSpringConstant=500
    bAngularSpring=true
    bOverrideAngularDamping=true
    bUseGravityOverride=true
    gravityScale=0.35
    simulationSpace=RootRelative
    alphaBoolBlend {blendOption=Linear}
    boundBone {boneName=dyn_hood_string_01_r}
    componentAppliedLinearAccClamp=<5000 5000 5000>
    componentLinearAccScale=<0.6 0.8 0.6>
    componentLinearVelScale=<1 1.2 1>
    gravityOverride {z=-1500}
    physicsBodyDefinitions [{
      boundBone {boneName=dyn_hood_string_01_r}
      boxExtents=<10 10 10>
      constraintSetup {angularLimitsMax=<1 0 20>, angularLimitsMin=<-1 -45 -20>, angularTarget=<1 0 0>}
      localJointOffset=<4 0 0>
     }]
   }
   n4 = Anim Dynamics - Bone: dyn_hood_string_02_r!(SimulationSpace:AnimPhysSimSpaceType Enum, ComponentPose=@n3, Alpha=1, ExternalForce="0,0,0") -> (Pose) {
    angularDampingOverride=1
    angularSpringConstant=100
    bAngularSpring=true
    bOverrideAngularDamping=true
    bUseGravityOverride=true
    gravityScale=0.35
    simulationSpace=RootRelative
    alphaBoolBlend {blendOption=Linear}
    boundBone {boneName=dyn_hood_string_02_r}
    componentAppliedLinearAccClamp=<5000 5000 5000>
    componentLinearAccScale=<0.6 0.8 0.6>
    componentLinearVelScale=<1 1.2 1>
    gravityOverride {z=-1500}
    physicsBodyDefinitions [{
      boundBone {boneName=dyn_hood_string_02_r}
      boxExtents=<15 15 15>
      constraintSetup {angularLimitsMax=<1 -3 20>, angularLimitsMin=<-1 -45 -20>, angularTarget=<1 0 0>}
      localJointOffset=<4 0 0>
     }]
   }
   n5 = Anim Dynamics - Bone: dyn_hood_string_03_r!(SimulationSpace:AnimPhysSimSpaceType Enum, ComponentPose=@n4, Alpha=1, ExternalForce="0,0,0") -> (Pose) {
    angularDampingOverride=1
    angularSpringConstant=100
    bAngularSpring=true
    bOverrideAngularDamping=true
    bUseGravityOverride=true
    gravityScale=0.35
    simulationSpace=RootRelative
    alphaBoolBlend {blendOption=Linear}
    boundBone {boneName=dyn_hood_string_03_r}
    componentAppliedLinearAccClamp=<5000 5000 5000>
    componentLinearAccScale=<0.6 0.8 0.6>
    componentLinearVelScale=<1 1.2 1>
    gravityOverride {z=-1500}
    physicsBodyDefinitions [{
      boundBone {boneName=dyn_hood_string_03_r}
      boxExtents=<15 15 15>
      constraintSetup {angularLimitsMax=<1 -3 45>, angularLimitsMin=<-1 -45 -45>, angularTarget=<1 0 0>}
      localJointOffset=<4 0 0>
     }]
   }
   n6 = Anim Dynamics - Bone: dyn_hood_string_04_r!(SimulationSpace:AnimPhysSimSpaceType Enum, ComponentPose=@n5, Alpha=1, ExternalForce="0,0,0") -> (Pose) {
    angularDampingOverride=1
    angularSpringConstant=100
    bAngularSpring=true
    bOverrideAngularDamping=true
    bUseGravityOverride=true
    gravityScale=0.35
    simulationSpace=RootRelative
    alphaBoolBlend {blendOption=Linear}
    boundBone {boneName=dyn_hood_string_04_r}
    componentAppliedLinearAccClamp=<5000 5000 5000>
    componentLinearAccScale=<0.6 0.8 0.6>
    componentLinearVelScale=<1 1.2 1>
    gravityOverride {z=-1500}
    physicsBodyDefinitions [{
      boundBone {boneName=dyn_hood_string_04_r}
      boxExtents=<15 15 15>
      constraintSetup {angularLimitsMax=<1 -2 45>, angularLimitsMin=<-1 -45 -45>, angularTarget=<1 0 0>}
      localJointOffset=<4 0 0>
     }]
   }
   n7 = Anim Dynamics - Bone: dyn_hood_string_05_r!(SimulationSpace:AnimPhysSimSpaceType Enum, ComponentPose=@n6, Alpha=1, ExternalForce="0,0,0") -> (Pose) {
    angularDampingOverride=1
    angularSpringConstant=100
    bAngularSpring=true
    bOverrideAngularDamping=true
    bUseGravityOverride=true
    gravityScale=0.35
    simulationSpace=RootRelative
    alphaBoolBlend {blendOption=Linear}
    boundBone {boneName=dyn_hood_string_05_r}
    componentAppliedLinearAccClamp=<5000 5000 5000>
    componentLinearAccScale=<0.6 0.8 0.6>
    componentLinearVelScale=<1 1.2 1>
    gravityOverride {z=-1500}
    physicsBodyDefinitions [{
      boundBone {boneName=dyn_hood_string_05_r}
      boxExtents=<20 20 20>
      constraintSetup {angularLimitsMax=<1 -2 45>, angularLimitsMin=<-1 -45 -45>, angularTarget=<1 0 0>}
      localJointOffset=<4 0 0>
     }]
   }
   n8 = Anim Dynamics - Bone: dyn_hood_string_01_l!(SimulationSpace:AnimPhysSimSpaceType Enum, ComponentPose=@n7, Alpha=1, ExternalForce="0,0,0") -> (Pose) {
    angularDampingOverride=1
    angularSpringConstant=500
    bAngularSpring=true
    bOverrideAngularDamping=true
    bUseGravityOverride=true
    gravityScale=0.35
    simulationSpace=RootRelative
    alphaBoolBlend {blendOption=Linear}
    boundBone {boneName=dyn_hood_string_01_l}
    componentAppliedLinearAccClamp=<10000 10000 10000>
    componentLinearAccScale=<0.6 0.8 0.6>
    componentLinearVelScale=<1 1.2 1>
    gravityOverride {z=-1500}
    physicsBodyDefinitions [{
      boundBone {boneName=dyn_hood_string_01_l}
      boxExtents=<10 10 10>
      constraintSetup {angularLimitsMax=<1 0 30>, angularLimitsMin=<-1 -45 -30>, angularTarget=<1 0 0>}
      localJointOffset=<4 0 0>
     }]
   }
   n9 = Anim Dynamics - Bone: dyn_hood_string__02_l!(SimulationSpace:AnimPhysSimSpaceType Enum, ComponentPose=@n8, Alpha=1, ExternalForce="0,0,0") -> (Pose) {
    angularDampingOverride=1
    angularSpringConstant=100
    bAngularSpring=true
    bOverrideAngularDamping=true
    bUseGravityOverride=true
    gravityScale=0.35
    simulationSpace=RootRelative
    alphaBoolBlend {blendOption=Linear}
    boundBone {boneName=dyn_hood_string__02_l}
    componentAppliedLinearAccClamp=<10000 10000 10000>
    componentLinearAccScale=<0.6 0.8 0.6>
    componentLinearVelScale=<1 1.2 1>
    gravityOverride {z=-1500}
    physicsBodyDefinitions [{
      boundBone {boneName=dyn_hood_string__02_l}
      boxExtents=<20 20 20>
      constraintSetup {angularLimitsMax=<1 -2 30>, angularLimitsMin=<-1 -45 -30>, angularTarget=<1 0 0>}
      localJointOffset=<4 0 0>
     }]
   }
   n10 = Anim Dynamics - Bone: dyn_hood_string_03_l!(SimulationSpace:AnimPhysSimSpaceType Enum, ComponentPose=@n9, Alpha=1, ExternalForce="0,0,0") -> (Pose) {
    angularDampingOverride=1
    angularSpringConstant=100
    bAngularSpring=true
    bOverrideAngularDamping=true
    bUseGravityOverride=true
    gravityScale=0.35
    simulationSpace=RootRelative
    alphaBoolBlend {blendOption=Linear}
    boundBone {boneName=dyn_hood_string_03_l}
    componentAppliedLinearAccClamp=<10000 10000 10000>
    componentLinearAccScale=<0.6 0.8 0.6>
    componentLinearVelScale=<1 1.2 1>
    gravityOverride {z=-1500}
    physicsBodyDefinitions [{
      boundBone {boneName=dyn_hood_string_03_l}
      boxExtents=<20 20 20>
      constraintSetup {angularLimitsMax=<1 -3 60>, angularLimitsMin=<-1 -45 -60>, angularTarget=<1 0 0>}
      localJointOffset=<4 0 0>
     }]
   }
   n11 = Anim Dynamics - Bone: dyn_hood_string_04_l!(SimulationSpace:AnimPhysSimSpaceType Enum, ComponentPose=@n10, Alpha=1, ExternalForce="0,0,0") -> (Pose) {
    angularDampingOverride=1
    angularSpringConstant=100
    bAngularSpring=true
    bOverrideAngularDamping=true
    bUseGravityOverride=true
    gravityScale=0.35
    simulationSpace=RootRelative
    alphaBoolBlend {blendOption=Linear}
    boundBone {boneName=dyn_hood_string_04_l}
    componentAppliedLinearAccClamp=<10000 10000 10000>
    componentLinearAccScale=<0.6 0.8 0.6>
    componentLinearVelScale=<1 1.2 1>
    gravityOverride {z=-1500}
    physicsBodyDefinitions [{
      boundBone {boneName=dyn_hood_string_04_l}
      boxExtents=<20 20 20>
      constraintSetup {angularLimitsMax=<1 -2.5 40>, angularLimitsMin=<-1 -45 -40>, angularTarget=<1 0 0>}
      localJointOffset=<4 0 0>
     }]
   }
   n12 = Anim Dynamics - Bone: dyn_hood_string_05_l!(SimulationSpace:AnimPhysSimSpaceType Enum, ComponentPose=@n11, Alpha=1, ExternalForce="0,0,0") -> (Pose) {
    angularDampingOverride=1
    angularSpringConstant=100
    bAngularSpring=true
    bOverrideAngularDamping=true
    bUseGravityOverride=true
    gravityScale=0.35
    simulationSpace=RootRelative
    alphaBoolBlend {blendOption=Linear}
    boundBone {boneName=dyn_hood_string_05_l}
    componentAppliedLinearAccClamp=<10000 10000 10000>
    componentLinearAccScale=<0.6 0.8 0.6>
    componentLinearVelScale=<1 1.2 1>
    gravityOverride {z=-1500}
    physicsBodyDefinitions [{
      boundBone {boneName=dyn_hood_string_05_l}
      boxExtents=<20 20 20>
      constraintSetup {angularLimitsMax=<1 -2 40>, angularLimitsMin=<-1 -45 -40>, angularTarget=<1 0 0>}
      localJointOffset=<4 0 0>
     }]
   }
   n13 = Anim Dynamics - Bone: dyn_hood_01(ComponentPose=@n12, Alpha=@n1, ExternalForce="0,0,0") {
    angularDampingOverride=1
    angularSpringConstant=500
    bAngularSpring=true
    bLinearSpring=true
    bOverrideAngularDamping=true
    bOverrideLinearDamping=true
    bUseGravityOverride=true
    gravityScale=0.35
    linearDampingOverride=0.9999
    linearSpringConstant=50
    simulationSpace=RootRelative
    alphaBoolBlend {blendOption=Linear}
    boundBone {boneName=dyn_hood_01}
    componentLinearAccScale=<1 1 1>
    componentLinearVelScale=<1 1 1>
    gravityOverride {z=-2250}
    physicsBodyDefinitions [{
      boundBone {boneName=dyn_hood_01}
      boxExtents=<60 60 60>
      constraintSetup {angularLimitsMax=<1 15 5>, angularLimitsMin=<-1 -60 -5>, angularTarget=<1 0 0>, linearAxesMin=<-6 0 0>}
      localJointOffset=<4 0 0>
     }]
   }
   n14 = Anim Dynamics - Bone: dyn_hood_02(ComponentPose=@n13, Alpha=@n1, ExternalForce="0,0,0") {
    angularDampingOverride=1
    angularSpringConstant=500
    bAngularSpring=true
    bLinearSpring=true
    bOverrideAngularDamping=true
    bOverrideLinearDamping=true
    bUseGravityOverride=true
    gravityScale=0.35
    linearDampingOverride=0.9999
    linearSpringConstant=500
    simulationSpace=RootRelative
    alphaBoolBlend {blendOption=Linear}
    boundBone {boneName=dyn_hood_02}
    componentLinearAccScale=<1 1 1>
    componentLinearVelScale=<1 1 1>
    gravityOverride {z=-2500}
    physicsBodyDefinitions [{
      boundBone {boneName=dyn_hood_02}
      boxExtents=<60 60 60>
      constraintSetup {angularLimitsMax=<1 25 5>, angularLimitsMin=<-1 -80 -5>, angularTarget=<1 0 0>, linearAxesMax=<0 0 0.5>, linearAxesMin=<-1 0 0>}
      localJointOffset=<4 0 0>
     }]
   }
   n15 = Anim Dynamics - Bone: dyn_hood_shoulder_l(ComponentPose=@n14, Alpha=@n1, ExternalForce="0,0,0") {
    angularDampingOverride=1
    angularSpringConstant=1000
    bAngularSpring=true
    bLinearSpring=true
    bOverrideAngularDamping=true
    bOverrideLinearDamping=true
    bUseGravityOverride=true
    gravityScale=0.35
    linearDampingOverride=0.9999
    linearSpringConstant=50
    simulationSpace=RootRelative
    alphaBoolBlend {blendOption=Linear}
    boundBone {boneName=dyn_hood_shoulder_l}
    componentLinearAccScale=<1 1 1>
    componentLinearVelScale=<1 1 1>
    gravityOverride {z=-2500}
    physicsBodyDefinitions [{
      boundBone {boneName=dyn_hood_shoulder_l}
      boxExtents=<60 60 60>
      constraintSetup {angularLimitsMax=<1 2 1>, angularLimitsMin=<-1 -2 -1>, angularTarget=<1 0 0>, linearAxesMax=<0 0 5>}
      localJointOffset=<4 0 0>
     }]
   }
   n16 = Anim Dynamics - Bone: dyn_hood_shoulder_r(ComponentPose=@n15, Alpha=@n1, ExternalForce="0,0,0") {
    angularDampingOverride=1
    angularSpringConstant=1000
    bAngularSpring=true
    bLinearSpring=true
    bOverrideAngularDamping=true
    bOverrideLinearDamping=true
    bUseGravityOverride=true
    gravityScale=0.35
    linearDampingOverride=0.9999
    linearSpringConstant=50
    simulationSpace=RootRelative
    alphaBoolBlend {blendOption=Linear}
    boundBone {boneName=dyn_hood_shoulder_r}
    componentLinearAccScale=<1 1 1>
    componentLinearVelScale=<1 1 1>
    gravityOverride {z=-2500}
    physicsBodyDefinitions [{
      boundBone {boneName=dyn_hood_shoulder_r}
      boxExtents=<60 60 60>
      constraintSetup {angularLimitsMax=<1 2 1>, angularLimitsMin=<-1 -2 -1>, angularTarget=<1 0 0>, linearAxesMin=<0 0 -5>}
      localJointOffset=<4 0 0>
     }]
   }
   n17 = "Transform (Modify) Bone - Bone: dyn_hood_01"!(ComponentPose=@n16, bAlphaBoolEnabled:bool=Get Is Wearing Backbling(), Translation="0,5,0", Rotation="0,0,0", Scale="1,1,1") -> (Pose) {
    alphaInputType=Bool
    translationMode=BMM_Additive
    alphaBoolBlend {blendOption=Linear}
    boneToModify {boneName=dyn_hood_01}
    translation {y=5}
   }
   Output Pose(Result=Component To Local(ComponentPose=@n17)) {layerGroup=None}
  ]}
 EventGraph {nodes [
   Event OnCharacterCustomizationLoaded()
   > Set Is Wearing Backbling(Is Wearing Backbling=CheckForBackbling())
   > Branch(Condition=Get bIsInFrontEnd())
   > Set Sim Space(Sim Space:AnimPhysSimSpaceType Enum=World) -> (Output_Get:AnimPhysSimSpaceType Enum)
   Event OnBeginSkydiving()
   > Set AnimDyn Strength(AnimDyn Strength:double=0.4) -> (Output_Get:double)
   Event OnEndSkydiving()
   > Set AnimDyn Strength(AnimDyn Strength=1)
  ]}
]
variables [
 AnimDyn Strength {type=number}
 Is Wearing Backbling {type=boolean}
 Sim Space {type=AnimPhysSimSpaceType, default=RootRelative}
]
