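# Post-copy AnimBP for the FossilMech outfit's face accessory, layering jaw and aiming tweaks onto poses copied from other character meshes.
# The attached parent's pose and curves form the base, with the left hood-shoulder branch blended from the body mesh cached on customization load.
# Curves jaw_open_pose and jaw_up_pose drive additive jaw rotations.
# Targeting sets a weapon-class-dependent, interpolated head roll correction, cleared when targeting ends.
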AnimBlueprint /BRCosmetics/Characters/Player/Male/Medium/Bodies/M_MED_FossilMech/Meshes/Parts/M_MED_FossilMech_FaceAcc_AnimBP
target_skeleton=&/M_MED_FossilMech_FaceAcc_Skeleton
graphs [
 AnimGraph {nodes [
   n1 = Copy Pose From Mesh(SourceMeshComponent=Get Body Mesh() -> (Body Mesh:Skeletal Mesh Component Object Reference)) {bUseAttachedParent=true}
   n2 = Copy Pose From Mesh() {bCopyCurves=true, bUseAttachedParent=true}
   n3 = "Transform (Modify) Bone - Bone: dyn_jaw"!(
    ComponentPose=Local To Component(LocalPose=Layered blend per bone(BasePose=@n2, BlendPoses_0=@n1) {layerSetup [{branchFilters [{boneName=dyn_hood_shoulder_base_l}]}]})
    AlphaCurveName:Name=jaw_open_pose
    Translation="0,0,0"
    Rotation="-25,0,0"
    Scale="1,1,1"
   ) -> (Pose) {
    alphaInputType=Curve
    rotationMode=BMM_Additive
    rotationSpace=BCS_BoneSpace
    alphaBoolBlend {blendOption=Linear}
    boneToModify {boneName=dyn_jaw}
    rotation {pitch=-25}
   }
   n4 = "Transform (Modify) Bone - Bone: dyn_jaw"(ComponentPose=@n3, AlphaCurveName=jaw_up_pose, Translation="0,0,0", Rotation="10,0,0", Scale="1,1,1") {
    alphaInputType=Curve
    rotationMode=BMM_Additive
    rotationSpace=BCS_BoneSpace
    alphaBoolBlend {blendOption=Linear}
    boneToModify {boneName=dyn_jaw}
    rotation {pitch=10}
   }
   n5 = "Transform (Modify) Bone - Bone: head"(ComponentPose=@n4, Alpha=Get Gun Correction() -> (Gun Correction:double), Translation="0,0,0", Rotation="0,0,20", Scale="1,1,1") {
    rotationMode=BMM_Additive
    rotationSpace=BCS_BoneSpace
    alphaBoolBlend {blendOption=Linear}
    alphaScaleBiasClamp {bInterpResult=true}
    boneToModify {boneName=head}
    rotation {roll=20}
   }
   Output Pose(Result=Component To Local(ComponentPose=@n5)) {layerGroup=None}
  ]}
 EventGraph {nodes [
   n1 = Event OnCharacterCustomizationLoaded()
   > Set Body Mesh(Body Mesh:Skeletal Mesh Component Object Reference=Get Array Item with EFort Custom Part Type(Array=@n1.CustomMeshParts)) -> (Output_Get:Skeletal Mesh Component Object Reference)
   Event OnBeginTargeting() {is=title K2Node_Event}
   n2 = GetCoreAnimation!(self:Fort Weapon Object Reference=Get CurrentWeapon!(self:Fort Pawn Object Reference=TryGetFortPawn()) -> (CurrentWeapon:Fort Weapon Object Reference)) -> (ReturnValue:EFortWeaponCoreAnimation Enum)
   > n3 = Branch(Condition=OR Boolean(A="Equal (Enum)"(A:EFortWeaponCoreAnimation Enum=@n2, B:EFortWeaponCoreAnimation Enum=AssaultRifle), B="Equal (Enum)"(A=@n2, B=Rifle)))
   > n4 = Set Gun Correction(Gun Correction:double=0.8) -> (Output_Get:double)
   Do Once(execute=@n8.then, Reset=[@n4.then, @n7.then, @n6.then])
   > Set Gun Correction(Gun Correction=0)
   n5 = Branch(execute=@n3.else, Condition=OR Boolean(A="Equal (Enum)"(A=@n2, B=Shotgun), B="Equal (Enum)"(A=@n2, B=TacticalShotgun)))
   > n6 = Set Gun Correction(Gun Correction=1)
   Branch(
    execute=@n5.else
    Condition=OR Boolean!(A="Equal (Enum)"(A=@n2, B=Pistol), B="Equal (Enum)"(A=@n2, B=MachinePistol), C:bool="Equal (Enum)"(A=@n2, B=SMG_P90)) -> (ReturnValue)
   )
   > n7 = Set Gun Correction(Gun Correction=0.4)
   n8 = Event OnEndTargeting() {is=title K2Node_Event}
  ]}
]
variables [
 Body Mesh {type=SkeletalMeshComponent}
 Gun Correction {type=number}
 JawOpenPose {type=number}
 JawUpPose {type=number}
]
